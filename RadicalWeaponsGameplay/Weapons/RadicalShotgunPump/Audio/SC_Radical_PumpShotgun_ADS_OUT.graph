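# Sound cue for the Radical pump shotgun's aim-down-sights exit foley.
# A single wave player feeds a modulator that applies slight random pitch variation.
# It plays quietly through the weapon foley sound class, using the shared weapon reload attenuation settings.
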SoundCue /RadicalWeaponsGameplay/Weapons/RadicalShotgunPump/Audio/SC_Radical_PumpShotgun_ADS_OUT
root=Modulator_1
nodes [
 {id=Modulator_1, children [WavePlayer_0], properties {pitchMax=1.02, pitchMin=0.98}}
 {
  id=WavePlayer_0
  properties {soundWaveAssetPtr=/RadicalWeaponsGameplay/Weapons/RadicalShotgunPump/Audio/SoundWaves/RadicalPumpShotgun_ADS_Out_01}
 }
]
properties {
 attenuationSettings=&Fort_Weapon_Reload_Att
 maxDistance=2250
 soundClassObject=&Fort_WeaponFoley
 volumeMultiplier=0.1
}
