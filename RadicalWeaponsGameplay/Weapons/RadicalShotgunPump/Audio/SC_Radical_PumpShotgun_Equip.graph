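# Sound cue that plays the equip foley for the Radical pump shotgun.
# A single equip wave runs through a short fixed delay into a root Modulator left at default settings.
# It uses the shared weapon reload attenuation and weapon foley sound class, with reduced volume.
# Its concurrency override caps simultaneous plays at two, taking precedence over the assigned one-per-owner concurrency set.
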SoundCue /RadicalWeaponsGameplay/Weapons/RadicalShotgunPump/Audio/SC_Radical_PumpShotgun_Equip
root=Modulator_14
nodes [
 {id=Delay_0, children [WavePlayer_1], properties {delayMax=0.1, delayMin=0.1}}
 {id=Modulator_14, children [Delay_0]}
 {
  id=WavePlayer_1
  properties {soundWaveAssetPtr=/RadicalWeaponsGameplay/Weapons/RadicalShotgunPump/Audio/SoundWaves/RadicalPumpShotgun_Equip_01}
 }
]
properties {
 attenuationSettings=&Fort_Weapon_Reload_Att
 bOverrideConcurrency=true
 maxDistance=2250
 soundClassObject=&Fort_WeaponFoley
 volumeMultiplier=0.551
 attenuationOverrides {falloffDistance=1600}
 concurrencyOverrides {maxCount=2}
 concurrencySet [/Game/Sounds/Concurrency_Rules/OneMaximum_Owner_StopOldestRelease]
}
