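# Sound cue for the firing sound of the Radical pump shotgun, layered by distance.
# A quality node gives the first level five-variant random pools and shares a three-variant mix across the other three.
# Each mix layers close (1p shot plus charge-up), mid and far waves, each with distance attenuation and sound-class overrides.
# Far layers add slight random downward pitch, and local and global gunfire concurrency sets limit overlapping shots.
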SoundCue /RadicalWeaponsGameplay/Weapons/RadicalShotgunPump/Audio/SC_Radical_PumpShotgun_Fire
root=QualityLevel_0
nodes [
 {id=Attenuation_0, children [Random_8], properties {bOverrideAttenuation=true, attenuationOverrides {falloffDistance=1200}}}
 {
  id=Attenuation_1
  children [Modulator_1]
  properties {attenuationSettings=&Fort_Weapon_Projectile_Medium_Mono_Athena_Att}
 }
 {
  id=Attenuation_10
  children [Modulator_7]
  properties {attenuationSettings=&Fort_Weapon_Projectile_Distant_Mono_Athena_Att}
 }
 {
  id=Attenuation_11
  children [Modulator_8]
  properties {attenuationSettings=&Fort_Weapon_Projectile_Medium_Mono_Athena_Att}
 }
 {id=Attenuation_2, children [Mixer_2], properties {attenuationSettings=&Fort_Weapon_Projectile_Close_Athena_Att}}
 {
  id=Attenuation_3
  children [Modulator_6]
  properties {attenuationSettings=&Fort_Weapon_Projectile_Distant_Mono_Athena_Att}
 }
 {id=Attenuation_9, children [Mixer_7], properties {attenuationSettings=&Fort_Weapon_Projectile_Close_Athena_Att}}
 {id=Mixer_2, children [Random_0, Attenuation_0], properties {inputVolume [1.3,0.8]}}
 {id=Mixer_3, children [SoundClass_2,SoundClass_6,SoundClass_3], properties {inputVolume [0.8,0.8,0.7]}}
 {id=Mixer_6, children [SoundClass_9, SoundClass_11, SoundClass_10], properties {inputVolume [0.8,0.8,0.7]}}
 {id=Mixer_7, children [Random_10,Random_11], properties {inputVolume [1.3,0.8]}}
 {id=Modulator_1, children [Random_3], properties {volumeMax=2, volumeMin=2}}
 {id=Modulator_6, children [Random_4], properties {pitchMax=1, pitchMin=0.9, volumeMax=2, volumeMin=2}}
 {id=Modulator_7, children [Random_13], properties {pitchMax=1, pitchMin=0.9, volumeMax=2, volumeMin=2}}
 {id=Modulator_8, children [Random_12], properties {volumeMax=2, volumeMin=2}}
 {id=QualityLevel_0, children [Mixer_3,Mixer_6,Mixer_6,Mixer_6]}
 {
  id=Random_0
  children [WavePlayer_2, WavePlayer_5, WavePlayer_3, WavePlayer_4, WavePlayer_10]
  properties {bShouldExcludeFromBranchCulling=true, weights [1,1,1,1,1]}
 }
 {id=Random_10, children [WavePlayer_28, WavePlayer_29, WavePlayer_30], properties {bShouldExcludeFromBranchCulling=true, weights [1,1,1]}}
 {id=Random_11, children [WavePlayer_34, WavePlayer_35, WavePlayer_36], properties {bShouldExcludeFromBranchCulling=true, weights [1,1,1]}}
 {id=Random_12, children [WavePlayer_38, WavePlayer_39, WavePlayer_40], properties {bShouldExcludeFromBranchCulling=true, weights [1,1,1]}}
 {id=Random_13, children [WavePlayer_37, WavePlayer_41, WavePlayer_42], properties {bShouldExcludeFromBranchCulling=true, weights [1,1,1]}}
 {id=Random_3, children [WavePlayer_1, WavePlayer_13, WavePlayer_14, WavePlayer_6, WavePlayer_7], properties {weights [1,1,1,1,1]}}
 {id=Random_4, children [WavePlayer_0, WavePlayer_15, WavePlayer_16, WavePlayer_8, WavePlayer_9], properties {weights [1,1,1,1,1]}}
 {id=Random_8, children [WavePlayer_31, WavePlayer_32, WavePlayer_33], properties {weights [1,1,1]}}
 {
  id=SoundClass_10
  children [Attenuation_10]
  properties {soundClassOverride=&Fort_WeaponRanged_Fire_Distant}
 }
 {
  id=SoundClass_11
  children [Attenuation_11]
  properties {soundClassOverride=&Fort_WeaponRanged_Fire_Distant}
 }
 {
  id=SoundClass_2
  children [Attenuation_2]
  properties {soundClassOverride=&Fort_WeaponRanged_Fire_3P_PSM}
 }
 {
  id=SoundClass_3
  children [Attenuation_3]
  properties {soundClassOverride=&Fort_WeaponRanged_Fire_Distant}
 }
 {
  id=SoundClass_6
  children [Attenuation_1]
  properties {soundClassOverride=&Fort_WeaponRanged_Fire_Distant}
 }
 {
  id=SoundClass_9
  children [Attenuation_9]
  properties {soundClassOverride=&Fort_WeaponRanged_Fire_3P_PSM}
 }
 {
  id=WavePlayer_0
  properties {soundWaveAssetPtr=/RadicalWeaponsGameplay/Weapons/RadicalShotgunPump/Audio/SoundWaves/RadicalPumpShotgun_Fire_3P_Far_01}
 }
 {
  id=WavePlayer_1
  properties {soundWaveAssetPtr=/RadicalWeaponsGameplay/Weapons/RadicalShotgunPump/Audio/SoundWaves/RadicalPumpShotgun_Fire_3P_Mid_01}
 }
 {
  id=WavePlayer_10
  properties {soundWaveAssetPtr=/RadicalWeaponsGameplay/Weapons/RadicalShotgunPump/Audio/SoundWaves/RadicalPumpShotgun_Fire_1p_03}
 }
 {
  id=WavePlayer_13
  properties {soundWaveAssetPtr=/RadicalWeaponsGameplay/Weapons/RadicalShotgunPump/Audio/SoundWaves/RadicalPumpShotgun_Fire_3P_Mid_02}
 }
 {
  id=WavePlayer_14
  properties {soundWaveAssetPtr=/RadicalWeaponsGameplay/Weapons/RadicalShotgunPump/Audio/SoundWaves/RadicalPumpShotgun_Fire_3P_Mid_03}
 }
 {
  id=WavePlayer_15
  properties {soundWaveAssetPtr=/RadicalWeaponsGameplay/Weapons/RadicalShotgunPump/Audio/SoundWaves/RadicalPumpShotgun_Fire_3P_Far_03}
 }
 {
  id=WavePlayer_16
  properties {soundWaveAssetPtr=/RadicalWeaponsGameplay/Weapons/RadicalShotgunPump/Audio/SoundWaves/RadicalPumpShotgun_Fire_3P_Far_02}
 }
 {
  id=WavePlayer_2
  properties {soundWaveAssetPtr=/RadicalWeaponsGameplay/Weapons/RadicalShotgunPump/Audio/SoundWaves/RadicalPumpShotgun_Fire_1p_04}
 }
 {
  id=WavePlayer_28
  properties {soundWaveAssetPtr=/RadicalWeaponsGameplay/Weapons/RadicalShotgunPump/Audio/SoundWaves/RadicalPumpShotgun_Fire_1p_01}
 }
 {
  id=WavePlayer_29
  properties {soundWaveAssetPtr=/RadicalWeaponsGameplay/Weapons/RadicalShotgunPump/Audio/SoundWaves/RadicalPumpShotgun_Fire_1p_02}
 }
 {
  id=WavePlayer_3
  properties {soundWaveAssetPtr=&RadicalPumpShotgun_Fire_1p_01}
 }
 {
  id=WavePlayer_30
  properties {soundWaveAssetPtr=&RadicalPumpShotgun_Fire_1p_03}
 }
 {
  id=WavePlayer_31
  properties {soundWaveAssetPtr=/RadicalWeaponsGameplay/Weapons/RadicalShotgunPump/Audio/SoundWaves/RadicalPumpShotgun_Fire_1p_ChargeUp_01}
 }
 {
  id=WavePlayer_32
  properties {soundWaveAssetPtr=/RadicalWeaponsGameplay/Weapons/RadicalShotgunPump/Audio/SoundWaves/RadicalPumpShotgun_Fire_1p_ChargeUp_02}
 }
 {
  id=WavePlayer_33
  properties {soundWaveAssetPtr=/RadicalWeaponsGameplay/Weapons/RadicalShotgunPump/Audio/SoundWaves/RadicalPumpShotgun_Fire_1p_ChargeUp_03}
 }
 {
  id=WavePlayer_34
  properties {soundWaveAssetPtr=&RadicalPumpShotgun_Fire_1p_ChargeUp_01}
 }
 {
  id=WavePlayer_35
  properties {soundWaveAssetPtr=&RadicalPumpShotgun_Fire_1p_ChargeUp_02}
 }
 {
  id=WavePlayer_36
  properties {soundWaveAssetPtr=&RadicalPumpShotgun_Fire_1p_ChargeUp_03}
 }
 {
  id=WavePlayer_37
  properties {soundWaveAssetPtr=&RadicalPumpShotgun_Fire_3P_Far_01}
 }
 {
  id=WavePlayer_38
  properties {soundWaveAssetPtr=&RadicalPumpShotgun_Fire_3P_Mid_01}
 }
 {
  id=WavePlayer_39
  properties {soundWaveAssetPtr=&RadicalPumpShotgun_Fire_3P_Mid_02}
 }
 {
  id=WavePlayer_4
  properties {soundWaveAssetPtr=&RadicalPumpShotgun_Fire_1p_02}
 }
 {
  id=WavePlayer_40
  properties {soundWaveAssetPtr=&RadicalPumpShotgun_Fire_3P_Mid_03}
 }
 {
  id=WavePlayer_41
  properties {soundWaveAssetPtr=&RadicalPumpShotgun_Fire_3P_Far_03}
 }
 {
  id=WavePlayer_42
  properties {soundWaveAssetPtr=&RadicalPumpShotgun_Fire_3P_Far_02}
 }
 {
  id=WavePlayer_5
  properties {soundWaveAssetPtr=/RadicalWeaponsGameplay/Weapons/RadicalShotgunPump/Audio/SoundWaves/RadicalPumpShotgun_Fire_1p_05}
 }
 {
  id=WavePlayer_6
  properties {soundWaveAssetPtr=/RadicalWeaponsGameplay/Weapons/RadicalShotgunPump/Audio/SoundWaves/RadicalPumpShotgun_Fire_3P_Mid_05}
 }
 {
  id=WavePlayer_7
  properties {soundWaveAssetPtr=/RadicalWeaponsGameplay/Weapons/RadicalShotgunPump/Audio/SoundWaves/RadicalPumpShotgun_Fire_3P_Mid_04}
 }
 {
  id=WavePlayer_8
  properties {soundWaveAssetPtr=/RadicalWeaponsGameplay/Weapons/RadicalShotgunPump/Audio/SoundWaves/RadicalPumpShotgun_Fire_3P_Far_04}
 }
 {
  id=WavePlayer_9
  properties {soundWaveAssetPtr=/RadicalWeaponsGameplay/Weapons/RadicalShotgunPump/Audio/SoundWaves/RadicalPumpShotgun_Fire_3P_Far_05}
 }
]
properties {
 maxDistance=27500
 volumeMultiplier=0.7998
 attenuationOverrides {falloffDistance=5000, attenuationShapeExtents {x=1000}}
 concurrencyOverrides {bLimitToOwner=true, maxCount=3}
 concurrencySet [&Weapon_Gunfire_Local_Conc, &Weapon_Gunfire_Global_Conc]
}
