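# First-person fire sound cue for the Radical pump shotgun.
# Each shot mixes random fire, charge-up and mechanical layers with a delayed, pitched-down shell impact and a sub sweetener.
# Continuous modulators on the IsADS parameter scale the mechanical/charge-up, shell and sub layers when aiming down sights.
# A quality-level switch selects between a fuller branch and a lighter one with fewer random variants, both using the first-person fire sound class.
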SoundCue /RadicalWeaponsGameplay/Weapons/RadicalShotgunPump/Audio/SC_Radical_PumpShotgun_Fire_1P
root=QualityLevel_0
nodes [
 {id=Delay_0, children [Modulator_2], properties {delayMax=0.8, delayMin=0.8}}
 {id=Delay_3, children [Modulator_7], properties {delayMax=0.8, delayMin=0.8}}
 {id=Mixer_1, children [Random_7,Random_9], properties {inputVolume [0.6,1]}}
 {
  id=Mixer_2
  children [Modulator_0, ModulatorContinuous_0, ModulatorContinuous_2, null, ModulatorContinuous_5]
  properties {inputVolume [1,1.3,0.7,1,1]}
 }
 {id=Mixer_3, children [Random_4,Random_13], properties {inputVolume [0.6,1]}}
 {
  id=Mixer_4
  children [Modulator_3, ModulatorContinuous_3, ModulatorContinuous_11, ModulatorContinuous_12]
  properties {inputVolume [1,1.3,0.7,1]}
 }
 {
  id=ModulatorContinuous_0
  children [Mixer_3]
  properties {volumeModulationParams {default=0, maxOutput=1.5, minOutput=1, paramMode=MPM_Abs, parameterName=IsADS}}
 }
 {
  id=ModulatorContinuous_11
  children [Delay_3]
  properties {volumeModulationParams {default=0, maxOutput=1.75, minOutput=1, paramMode=MPM_Abs, parameterName=IsADS}}
 }
 {
  id=ModulatorContinuous_12
  children [WavePlayer_57]
  properties {volumeModulationParams {default=0, maxOutput=0.5, paramMode=MPM_Abs, parameterName=IsADS}}
 }
 {
  id=ModulatorContinuous_2
  children [Delay_0]
  properties {volumeModulationParams {default=0, maxOutput=1.75, minOutput=1, paramMode=MPM_Abs, parameterName=IsADS}}
 }
 {
  id=ModulatorContinuous_3
  children [Mixer_1]
  properties {volumeModulationParams {default=0, maxOutput=1.5, minOutput=1, paramMode=MPM_Abs, parameterName=IsADS}}
 }
 {
  id=ModulatorContinuous_5
  children [WavePlayer_11]
  properties {volumeModulationParams {default=0, maxOutput=0.5, paramMode=MPM_Abs, parameterName=IsADS}}
 }
 {id=Modulator_0, children [Random_1], properties {pitchMax=1, pitchMin=1}}
 {id=Modulator_2, children [Random_8], properties {pitchMax=0.95, pitchMin=0.85}}
 {id=Modulator_3, children [Random_6], properties {pitchMax=1, pitchMin=1}}
 {id=Modulator_7, children [Random_16], properties {pitchMax=0.95, pitchMin=0.85}}
 {id=QualityLevel_0, children [SoundClass_1,SoundClass_1,SoundClass_0,SoundClass_0]}
 {
  id=Random_1
  children [WavePlayer_5, WavePlayer_6, WavePlayer_8, WavePlayer_9, WavePlayer_10]
  properties {bShouldExcludeFromBranchCulling=true, weights [1,1,1,1,1]}
 }
 {id=Random_13, children [WavePlayer_51, WavePlayer_52, WavePlayer_53], properties {bShouldExcludeFromBranchCulling=true, weights [1,1,1]}}
 {id=Random_16, children [WavePlayer_61, WavePlayer_62, WavePlayer_63], properties {bShouldExcludeFromBranchCulling=true, weights [1,1,1]}}
 {id=Random_4, children [WavePlayer_31, WavePlayer_32, WavePlayer_33], properties {bShouldExcludeFromBranchCulling=true, weights [1,1,1]}}
 {id=Random_6, children [WavePlayer_19, WavePlayer_20, WavePlayer_18], properties {bShouldExcludeFromBranchCulling=true, weights [1,1,1]}}
 {id=Random_7, children [WavePlayer_21, WavePlayer_22], properties {bShouldExcludeFromBranchCulling=true, weights [1,1]}}
 {
  id=Random_8
  children [WavePlayer_34, WavePlayer_35, WavePlayer_36, WavePlayer_37, WavePlayer_38]
  properties {bShouldExcludeFromBranchCulling=true, weights [1,1,1,1,1]}
 }
 {id=Random_9, children [WavePlayer_23, WavePlayer_24], properties {bShouldExcludeFromBranchCulling=true, weights [1,1]}}
 {id=SoundClass_0, children [Mixer_4], properties {soundClassOverride=&Fort_WeaponRanged_Fire_1P_PSM}}
 {id=SoundClass_1, children [Mixer_2], properties {soundClassOverride=&Fort_WeaponRanged_Fire_1P_PSM}}
 {
  id=WavePlayer_10
  properties {soundWaveAssetPtr=/RadicalWeaponsGameplay/Weapons/RadicalShotgunPump/Audio/SoundWaves/RadicalPumpShotgun_Fire_1p_03}
 }
 {
  id=WavePlayer_11
  properties {soundWaveAssetPtr=/RadicalWeaponsGameplay/Weapons/RadicalShotgunPump/Audio/SoundWaves/RadicalPumpShotgun_Fire_ADS_SUB_Sweetener}
 }
 {
  id=WavePlayer_18
  properties {soundWaveAssetPtr=/RadicalWeaponsGameplay/Weapons/RadicalShotgunPump/Audio/SoundWaves/RadicalPumpShotgun_Fire_1p_01}
 }
 {
  id=WavePlayer_19
  properties {soundWaveAssetPtr=/RadicalWeaponsGameplay/Weapons/RadicalShotgunPump/Audio/SoundWaves/RadicalPumpShotgun_Fire_1p_05}
 }
 {
  id=WavePlayer_20
  properties {soundWaveAssetPtr=/RadicalWeaponsGameplay/Weapons/RadicalShotgunPump/Audio/SoundWaves/RadicalPumpShotgun_Fire_1p_04}
 }
 {
  id=WavePlayer_21
  properties {soundWaveAssetPtr=/RadicalWeaponsGameplay/Weapons/RadicalShotgunPump/Audio/SoundWaves/RadicalPumpShotgun_Fire_1p_ChargeUp_01}
 }
 {
  id=WavePlayer_22
  properties {soundWaveAssetPtr=/RadicalWeaponsGameplay/Weapons/RadicalShotgunPump/Audio/SoundWaves/RadicalPumpShotgun_Fire_1p_ChargeUp_02}
 }
 {
  id=WavePlayer_23
  properties {soundWaveAssetPtr=/RadicalWeaponsGameplay/Weapons/RadicalShotgunPump/Audio/SoundWaves/RadicalPumpShotgun_Fire_1p_Mech_01}
 }
 {
  id=WavePlayer_24
  properties {soundWaveAssetPtr=/RadicalWeaponsGameplay/Weapons/RadicalShotgunPump/Audio/SoundWaves/RadicalPumpShotgun_Fire_1p_Mech_02}
 }
 {
  id=WavePlayer_31
  properties {soundWaveAssetPtr=&RadicalPumpShotgun_Fire_1p_ChargeUp_01}
 }
 {
  id=WavePlayer_32
  properties {soundWaveAssetPtr=&RadicalPumpShotgun_Fire_1p_ChargeUp_02}
 }
 {
  id=WavePlayer_33
  properties {soundWaveAssetPtr=/RadicalWeaponsGameplay/Weapons/RadicalShotgunPump/Audio/SoundWaves/RadicalPumpShotgun_Fire_1p_ChargeUp_03}
 }
 {
  id=WavePlayer_34
  properties {soundWaveAssetPtr=/RadicalWeaponsGameplay/Weapons/RadicalShotgunPump/Audio/SoundWaves/RadicalPumpShotgun_ShellImpact_01}
 }
 {
  id=WavePlayer_35
  properties {soundWaveAssetPtr=/RadicalWeaponsGameplay/Weapons/RadicalShotgunPump/Audio/SoundWaves/RadicalPumpShotgun_ShellImpact_02}
 }
 {
  id=WavePlayer_36
  properties {soundWaveAssetPtr=/RadicalWeaponsGameplay/Weapons/RadicalShotgunPump/Audio/SoundWaves/RadicalPumpShotgun_ShellImpact_03}
 }
 {
  id=WavePlayer_37
  properties {soundWaveAssetPtr=/RadicalWeaponsGameplay/Weapons/RadicalShotgunPump/Audio/SoundWaves/RadicalPumpShotgun_ShellImpact_04}
 }
 {
  id=WavePlayer_38
  properties {soundWaveAssetPtr=/RadicalWeaponsGameplay/Weapons/RadicalShotgunPump/Audio/SoundWaves/RadicalPumpShotgun_ShellImpact_05}
 }
 {
  id=WavePlayer_5
  properties {soundWaveAssetPtr=&RadicalPumpShotgun_Fire_1p_05}
 }
 {
  id=WavePlayer_51
  properties {soundWaveAssetPtr=&RadicalPumpShotgun_Fire_1p_Mech_01}
 }
 {
  id=WavePlayer_52
  properties {soundWaveAssetPtr=&RadicalPumpShotgun_Fire_1p_Mech_02}
 }
 {
  id=WavePlayer_53
  properties {soundWaveAssetPtr=/RadicalWeaponsGameplay/Weapons/RadicalShotgunPump/Audio/SoundWaves/RadicalPumpShotgun_Fire_1p_Mech_03}
 }
 {
  id=WavePlayer_57
  properties {soundWaveAssetPtr=&RadicalPumpShotgun_Fire_ADS_SUB_Sweetener}
 }
 {
  id=WavePlayer_6
  properties {soundWaveAssetPtr=&RadicalPumpShotgun_Fire_1p_04}
 }
 {
  id=WavePlayer_61
  properties {soundWaveAssetPtr=&RadicalPumpShotgun_ShellImpact_01}
 }
 {
  id=WavePlayer_62
  properties {soundWaveAssetPtr=&RadicalPumpShotgun_ShellImpact_02}
 }
 {
  id=WavePlayer_63
  properties {soundWaveAssetPtr=&RadicalPumpShotgun_ShellImpact_03}
 }
 {
  id=WavePlayer_8
  properties {soundWaveAssetPtr=&RadicalPumpShotgun_Fire_1p_01}
 }
 {
  id=WavePlayer_9
  properties {soundWaveAssetPtr=/RadicalWeaponsGameplay/Weapons/RadicalShotgunPump/Audio/SoundWaves/RadicalPumpShotgun_Fire_1p_02}
 }
]
properties {
 attenuationSettings=&Fort_Weapon_Projectile_FirstPerson_Athena_Att
 bDebug=true
 maxDistance=2097152
 soundClassObject=&Fort_WeaponRanged_Fire_1P_PSM
 volumeMultiplier=0.9
 attenuationOverrides {manualReverbSendLevel=-1.0165, reverbSendMethod=Manual}
 concurrencyOverrides {maxCount=100, resolutionRule=StopOldest, volumeScale=0.9}
 concurrencySet [&Weapon_Gunfire_FirstPerson_Conc]
}
