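# Sound cue for the Radical pump shotgun's low-ammo warning, routed to a degraded ranged-weapon low-ammo sound class.
# A mixer layers the low-ammo wave with the out-of-ammo wave, which is slightly delayed, pitched down and mixed louder.
# Delay and pitch are fixed, so playback does not vary.
# A concurrency override prevents new instances once the limit is reached.
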SoundCue /RadicalWeaponsGameplay/Weapons/RadicalShotgunPump/Audio/SC_Radical_PumpShotgun_LowAmmo
root=Mixer_0
nodes [
 {id=Delay_0, children [Modulator_0], properties {delayMax=0.15, delayMin=0.15}}
 {id=Mixer_0, children [WavePlayer_0,Delay_0], properties {inputVolume [1,1.3]}}
 {id=Modulator_0, children [WavePlayer_1], properties {pitchMax=0.7, pitchMin=0.7, volumeMax=1, volumeMin=1}}
 {
  id=WavePlayer_0
  properties {soundWaveAssetPtr=/RadicalWeaponsGameplay/Weapons/RadicalShotgunPump/Audio/SoundWaves/RadicalPumpShotgun_LowAmmo_01}
 }
 {
  id=WavePlayer_1
  properties {soundWaveAssetPtr=/RadicalWeaponsGameplay/Weapons/RadicalShotgunPump/Audio/SoundWaves/RadicalPumpShotgun_OutofAmmo_01}
 }
]
properties {
 bOverrideConcurrency=true
 maxDistance=2097152
 soundClassObject=/Game/Sounds/FortSoundClassesAndModes/Fort_WeaponRanged_LowAmmo_Degraded_PSM
 volumeMultiplier=1.5
 concurrencyOverrides {resolutionRule=PreventNew}
}
