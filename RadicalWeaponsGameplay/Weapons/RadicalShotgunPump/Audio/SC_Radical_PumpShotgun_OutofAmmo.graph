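# Sound cue for the Radical pump shotgun's out-of-ammo sound.
# A single wave player feeds a Modulator that slightly randomizes volume and holds pitch fixed.
# It routes to a third-person ranged weapon fire sound class at a reduced volume multiplier.
# A concurrency override caps simultaneous instances and stops the oldest when the limit is reached.
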SoundCue /RadicalWeaponsGameplay/Weapons/RadicalShotgunPump/Audio/SC_Radical_PumpShotgun_OutofAmmo
root=Modulator_0
nodes [
 {id=Modulator_0, children [WavePlayer_0], properties {pitchMax=1, pitchMin=1, volumeMax=1.1, volumeMin=1}}
 {
  id=WavePlayer_0
  properties {soundWaveAssetPtr=/RadicalWeaponsGameplay/Weapons/RadicalShotgunPump/Audio/SoundWaves/RadicalPumpShotgun_OutofAmmo_01}
 }
]
properties {
 bOverrideConcurrency=true
 maxDistance=2097152
 soundClassObject=&Fort_WeaponRanged_Fire_3P_PSM
 volumeMultiplier=0.65
 attenuationOverrides {falloffDistance=1600}
 concurrencyOverrides {maxCount=2, resolutionRule=StopOldest}
}
