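# Sound cue for the shell-insert sound when reloading the Radical pump shotgun.
# A Random node picks evenly among three cartridge-insert waves and feeds a Modulator whose pitch range is pinned, so pitch does not vary.
# It plays quietened as weapon foley with shared reload attenuation, and its concurrency override blocks new instances once the limit is reached.
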SoundCue /RadicalWeaponsGameplay/Weapons/RadicalShotgunPump/Audio/SC_Radical_PumpShotgun_ReloadInsert
root=Modulator_13
nodes [
 {id=Modulator_13, children [Random_2], properties {pitchMax=1, pitchMin=1}}
 {id=Random_2, children [WavePlayer_6,WavePlayer_7,WavePlayer_8], properties {weights [1,1,1]}}
 {
  id=WavePlayer_6
  properties {soundWaveAssetPtr=/RadicalWeaponsGameplay/Weapons/RadicalShotgunPump/Audio/SoundWaves/RadicalPumpShotgun_Reload_Cartridge_01}
 }
 {
  id=WavePlayer_7
  properties {soundWaveAssetPtr=/RadicalWeaponsGameplay/Weapons/RadicalShotgunPump/Audio/SoundWaves/RadicalPumpShotgun_Reload_Cartridge_02}
 }
 {
  id=WavePlayer_8
  properties {soundWaveAssetPtr=/RadicalWeaponsGameplay/Weapons/RadicalShotgunPump/Audio/SoundWaves/RadicalPumpShotgun_Reload_Cartridge_03}
 }
]
properties {
 attenuationSettings=&Fort_Weapon_Reload_Att
 bOverrideConcurrency=true
 maxDistance=2250
 soundClassObject=&Fort_WeaponFoley
 volumeMultiplier=0.501
 attenuationOverrides {falloffDistance=1600}
 concurrencyOverrides {resolutionRule=PreventNew}
}
