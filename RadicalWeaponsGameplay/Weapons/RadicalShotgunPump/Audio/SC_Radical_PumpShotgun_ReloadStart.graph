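# Sound cue for the reload start of the Radical pump shotgun, classed as weapon foley.
# The node graph is empty with no root, so the cue references no sound waves and plays nothing as shown.
# It keeps playback settings: reduced volume, the shared weapon reload attenuation with a falloff distance override, and a concurrency override that prevents new instances once the limit is reached.
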SoundCue /RadicalWeaponsGameplay/Weapons/RadicalShotgunPump/Audio/SC_Radical_PumpShotgun_ReloadStart
root=null
nodes []
properties {
 attenuationSettings=&Fort_Weapon_Reload_Att
 bOverrideConcurrency=true
 maxDistance=2000
 soundClassObject=&Fort_WeaponFoley
 volumeMultiplier=0.276
 attenuationOverrides {falloffDistance=1600}
 concurrencyOverrides {resolutionRule=PreventNew}
}
